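; Looping gameplay cue for the active heal effect in the CherrySmoke hero ranking system.
; On looping start, it runs the parent logic, then sets its Duration variable from the HF_HealDuration scalable float.
; If the first looping audio component is valid, it passes that value to it as the "Duration" audio parameter, so the sound knows the heal length.
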(event EventOnLoopingStart(EitherVFX-preferthisversion)
  (params (MyTarget :type Actor%20Object%20Reference) (Parameters :type Gameplay%20Cue%20Parameters%20Structure%20%28by%20ref%29) (ParticleComponents :type Array%20of%20FXSystem%20Component%20Object%20References) (AudioComponents :type Array%20of%20Audio%20Component%20Object%20References))
  (bind _output (Utilities|Array|Get(acopy) AudioComponents))
  (|CallParentFunction MyTarget Parameters ParticleComponents AudioComponents)
  (bind _output_get (Variables|Default|SetDuration (ScalableFloat|GetValueatLevel (Variables|Default|GetHF_HealDuration))))
  (Utilities|IsValid _output
    (:Is%20Valid
      (Audio|Parameter|SetFloatParameter _output "Duration" _output_get :!class /Script/AudioExtensions.AudioParameterControllerInterface))))

(fn ConstructionScript)
