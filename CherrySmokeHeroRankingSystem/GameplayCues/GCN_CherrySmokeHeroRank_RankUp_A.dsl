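; Burst gameplay cue that plays rank-up feedback for the CherrySmoke hero ranking system.
; It calls the parent burst, then truncates the cue's raw magnitude to an integer rank.
; That rank sets the "Rank" parameter on the first spawned particle component.
; If the target pawn is locally viewed, it also creates and plays a first-person 2D rank-up sound with the same "Rank" parameter.
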(fn OnBurst(EitherVFX-preferthisversion)
  (params (MyTarget :type Actor%20Object%20Reference) (Parameters :type Gameplay%20Cue%20Parameters%20Structure%20%28by%20ref%29) (ParticleComponents :type Array%20of%20FXSystem%20Component%20Object%20References) (AudioComponents :type Array%20of%20Audio%20Component%20Object%20References) (BurstCameraShakeInstance :type Camera%20Shake%20Base%20Object%20Reference) (BurstDecalInstance :type Decal%20Actor%20Object%20Reference))
  (bind (_normalizedmagnitude _rawmagnitude) (Ability|GameplayCue|BreakGameplayCueParameters Parameters))
  (bind _returnvalue (Math|Float|Truncate _rawmagnitude))
  (|CallParentFunction MyTarget Parameters ParticleComponents AudioComponents BurstCameraShakeInstance BurstDecalInstance)
  (Effects|Components|ParticleSystem|SetIntParameter (Utilities|Array|Get(acopy) ParticleComponents) "Rank" _returnvalue)
  (bind _asfort_pawn (Utilities|Casting|CastToFortPawn MyTarget))
  (if (Pawn|IsLocallyViewed _asfort_pawn)
    (bind _returnvalue_2 (Audio|CreateSound2D 0 "/CherrySmokeHeroRankingSystem/Assets/Audio/MSS_Ch6S3_CherryHeroRankingSystem_RankUp_1P.MSS_Ch6S3_CherryHeroRankingSystem_RankUp_1P"))
    (Audio|Parameter|SetIntegerParameter _returnvalue_2 "Rank" _returnvalue :!class /Script/AudioExtensions.AudioParameterControllerInterface)
    (Audio|Components|Audio|Play _returnvalue_2 0.0)
    (return)))
